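# Physics asset for the M_MED_PointWoof backpack cosmetic, defining collision bodies and joints for a chain of four dynamic bones on the pack.
# Capsule bodies cover the pack, a tiny base_bone and each chain bone, linked in sequence from pack to dyn_bone_04.
# The first and last chain joints lock twist and one swing axis to bend in a plane, middle joints have narrow angular limits, and the base-to-pack joint rotates freely.
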PhysicsAsset /BRCosmetics/Accessories/FORT_Backpacks/Backpack_M_MED_PointWoof/Meshes/M_MED_PointWoof_Pack_Physics
bodies {
 base_bone [{shapeType=Capsule, length=0.505, radius=0.505, shapeName=base_bone_capsule, rotation=<rot 0 -90 0>}]
 dyn_bone_01 [{shapeType=Capsule, length=18.8801, radius=2, shapeName=dyn_bone_01_capsule, center=<1.4632 0 0>, rotation=<rot 0 -90 0>}]
 dyn_bone_02 [{shapeType=Capsule, length=17.4051, radius=2, shapeName=dyn_bone_02_capsule, center=<1.4647 0 0>, rotation=<rot 0 -90 0>}]
 dyn_bone_03 [{shapeType=Capsule, length=15.8782, radius=2, shapeName=dyn_bone_03_capsule, center=<1.5095 0 0>, rotation=<rot 0 -90 0>}]
 dyn_bone_04 [{shapeType=Capsule, length=15.3209, radius=3.4744, shapeName=dyn_bone_04_capsule, rotation=<rot 0 -90 0>}]
 pack [{shapeType=Capsule, length=15.5, radius=15, shapeName=pack_capsule, center=<0 15.6808 -15>}]
}
constraints [
 {bone1Name=dyn_bone_02, bone2Name=dyn_bone_01, swing1LimitDegrees=20, swing2LimitDegrees=5, twistLimitDegrees=0}
 {bone1Name=dyn_bone_03, bone2Name=dyn_bone_02, swing1LimitDegrees=10, swing2LimitDegrees=5, twistLimitDegrees=10}
 {
  bone1Name=dyn_bone_04
  bone2Name=dyn_bone_03
  swing1LimitDegrees=180
  swing2LimitDegrees=15
  swing2Motion=Locked
  twistLimitDegrees=45
  twistMotion=Locked
 }
 {
  bone1Name=dyn_bone_01
  bone2Name=base_bone
  swing1LimitDegrees=55
  swing2LimitDegrees=15
  swing2Motion=Locked
  twistLimitDegrees=45
  twistMotion=Locked
 }
 {
  bone1Name=base_bone
  bone2Name=pack
  swing1LimitDegrees=45
  swing1Motion=Free
  swing2LimitDegrees=45
  swing2Motion=Free
  twistLimitDegrees=45
  twistMotion=Free
 }
]
